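# Physics asset for the dark Ghost Hunter female medium armor, simulating its cape and thigh straps.
# Three cape chains (left, middle, right) hang from master bones on spine_05 with tight swing limits, and cross-constraints tie both side-chain tips to the middle tip.
# Each thigh carries a pouch body with two three-link strap chains.
# Capsules on pelvis, spine, legs, arms and right hand form the body shape around these chains.
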PhysicsAsset /BRCosmetics/Characters/Player/Female/Medium/Bodies/F_MED_Ghost_Hunter_Armor_Dark/Meshes/F_MED_Ghost_Hunter_Armor_Dark_dyn_cape
bodies {
 calf_l [{shapeType=Capsule, length=39.7879, radius=7.1147, center=<-27.6451 -3.5904 -0.6294>, rotation=<rot 90 0 -2.1423>}]
 calf_r [{shapeType=Capsule, length=39.7879, radius=7.1147, center=<27.4315 3.5821 0.6294>, rotation=<rot 90 0 177.8577>}]
 dyn_cap_01_mid [{shapeType=Capsule, length=17, radius=11, shapeName=dyn_cap_a_01_l_capsule, center=<5 0 8>, rotation=<rot 0 0 90>}]
 dyn_cap_02_mid [{shapeType=Capsule, length=17, radius=11, shapeName=dyn_cap_a_02_l_capsule, center=<8 0 8>, rotation=<rot 0 0 90>}]
 dyn_cap_03_mid [{shapeType=Capsule, length=17, radius=11, shapeName=dyn_cap_a_03_l_capsule, center=<8 0 8>, rotation=<rot 0 0 90>}]
 dyn_cap_04_mid [{shapeType=Capsule, length=10, radius=11, shapeName=dyn_cap_a_04_l_capsule, center=<8 0 7>, rotation=<rot 90 0 0>}]
 dyn_cap_05_mid [{shapeType=Capsule, length=10, radius=11, shapeName=dyn_cap_a_04_l_capsule, center=<8 0 7>, rotation=<rot 90 0 0>}]
 dyn_cap_a_01_l [{shapeType=Capsule, length=16, radius=9, shapeName=dyn_cap_a_01_l_capsule, center=<6.5 0 5>, rotation=<rot 0 0 95>}]
 dyn_cap_a_01_r [{shapeType=Capsule, length=16, radius=9, shapeName=dyn_cap_a_01_l_capsule, center=<6.5 0 5>, rotation=<rot 0 0 84.9998>}]
 dyn_cap_a_02_l [{shapeType=Capsule, length=16, radius=9, shapeName=dyn_cap_a_02_l_capsule, center=<8 0 5>, rotation=<rot 0 0 90>}]
 dyn_cap_a_02_r [{shapeType=Capsule, length=16, radius=9, shapeName=dyn_cap_a_02_l_capsule, center=<8 0 5>, rotation=<rot 0 0 90>}]
 dyn_cap_a_03_l [{shapeType=Capsule, length=9, radius=9, shapeName=dyn_cap_a_03_l_capsule, center=<10 0 5>, rotation=<rot 90 0 0>}]
 dyn_cap_a_03_r [{shapeType=Capsule, length=9, radius=9, shapeName=dyn_cap_a_03_l_capsule, center=<10 0 5>, rotation=<rot 90 0 0>}]
 dyn_cap_a_04_l [{shapeType=Capsule, length=9, radius=9, shapeName=dyn_cap_a_03_l_capsule, center=<10 0 5>, rotation=<rot 90 0 0>}]
 dyn_cap_a_04_r [{shapeType=Capsule, length=9, radius=9, shapeName=dyn_cap_a_03_l_capsule, center=<10 0 5>, rotation=<rot 90 0 0>}]
 dyn_cap_a_l_MSTR [{shapeType=Capsule, length=0.505, radius=0.505, shapeName=dyn_cap_a_l_MSTR_capsule, rotation=<rot 0 -90 0>}]
 dyn_cap_a_r_MSTR [{shapeType=Capsule, length=0.505, radius=0.505, shapeName=dyn_cap_a_r_MSTR_capsule, rotation=<rot 0 -90 0>}]
 dyn_cap_mid_MSTR [{shapeType=Capsule, length=0.505, radius=0.505, shapeName=dyn_cap_mid_MSTR_capsule, rotation=<rot 0 -90 0>}]
 dyn_pouches_l_MSTR [{shapeType=Capsule, length=15.1986, radius=8.073, center=<9.8714 1.8509 -0.6162>, rotation=<rot 90 0 -9.5081>}]
 dyn_pouches_r_MSTR [{shapeType=Capsule, length=15.1986, radius=8.073, center=<9.8714 -1.8509 -0.6162>, rotation=<rot 90 0 10>}]
 dyn_straps_a_l_01 [{shapeType=Capsule, length=4, radius=2, shapeName=dyn_cap_a_02_l_capsule, center=<3 0 0>, rotation=<rot 90 0 0>}]
 dyn_straps_a_l_02 [{shapeType=Capsule, length=4, radius=2, shapeName=dyn_cap_a_02_l_capsule, center=<5 0 0>, rotation=<rot 90 0 0>}]
 dyn_straps_a_l_03 [{shapeType=Capsule, length=4, radius=2, shapeName=dyn_cap_a_02_l_capsule, center=<5 0 0>, rotation=<rot 90 0 0>}]
 dyn_straps_a_r_01 [{shapeType=Capsule, length=4, radius=2, shapeName=dyn_cap_a_02_l_capsule, center=<3 0 0>, rotation=<rot 90 0 0>}]
 dyn_straps_a_r_02 [{shapeType=Capsule, length=4, radius=2, shapeName=dyn_cap_a_02_l_capsule, center=<5 0 0>, rotation=<rot 90 0 0>}]
 dyn_straps_a_r_03 [{shapeType=Capsule, length=4, radius=2, shapeName=dyn_cap_a_02_l_capsule, center=<5 0 0>, rotation=<rot 90 0 0>}]
 dyn_straps_b_l_01 [{shapeType=Capsule, length=4, radius=2, shapeName=dyn_cap_a_02_l_capsule, center=<3 0 0>, rotation=<rot 90 0 0>}]
 dyn_straps_b_l_02 [{shapeType=Capsule, length=4, radius=2, shapeName=dyn_cap_a_02_l_capsule, center=<5 0 0>, rotation=<rot 90 0 0>}]
 dyn_straps_b_l_03 [{shapeType=Capsule, length=4, radius=2, shapeName=dyn_cap_a_02_l_capsule, center=<5 0 0>, rotation=<rot 90 0 0>}]
 dyn_straps_b_r_01 [{shapeType=Capsule, length=4, radius=2, shapeName=dyn_cap_a_02_l_capsule, center=<3 0 0>, rotation=<rot 90 0 0>}]
 dyn_straps_b_r_02 [{shapeType=Capsule, length=4, radius=2, shapeName=dyn_cap_a_02_l_capsule, center=<5 0 0>, rotation=<rot 90 0 0>}]
 dyn_straps_b_r_03 [{shapeType=Capsule, length=4, radius=2, shapeName=dyn_cap_a_02_l_capsule, center=<5 0 0>, rotation=<rot 90 0 0>}]
 hand_r [{shapeType=Capsule, length=2.6906, radius=5.9133, shapeName=upperarm_r_capsule, center=<-5.3292 0.0002 -0.0584>}]
 lowerarm_l [{
   shapeType=Capsule
   length=14.9406
   radius=4.1633
   shapeName=upperarm_r_capsule
   center=<11.8568 0.2347 -0.0004>
   rotation=<rot 90 0 173.1295>
  }]
 lowerarm_r [{
   shapeType=Capsule
   length=14.9406
   radius=4.1633
   shapeName=upperarm_r_capsule
   center=<-11.8568 -0.2347 0.0004>
   rotation=<rot 90 -90 -91.6498>
  }]
 pelvis [{shapeType=Capsule, length=9.2814, radius=11.2546, center=<-2.5744 -1.3468 -0.0002>}]
 spine_01 [{shapeType=Capsule, length=4.783, radius=9.7291, center=<0.7333 -0.5671 0.0003>}]
 spine_02 [{shapeType=Capsule, length=4.2727, radius=10.2394, center=<0.7333 0.0482 0.0003>}]
 spine_04 [{shapeType=Capsule, length=5.9469, radius=10.628, center=<-2.1985 0.0608 0>}]
 spine_05 [{shapeType=Capsule, length=14.9804, radius=8.9968, center=<0 0.3213 0.0001>}]
 thigh_l [{shapeType=Capsule, length=30.4486, radius=8.573, center=<-22.2474 0.3768 -1.241>, rotation=<rot 87.4927 173.7799 -179.9999>}]
 thigh_r [{shapeType=Capsule, length=30.4486, radius=8.573, center=<22.2474 -0.3768 1.241>, rotation=<rot 87.4927 173.7799 0.0001>}]
 upperarm_l [{
   shapeType=Capsule
   length=14.9406
   radius=4.1633
   shapeName=upperarm_r_capsule
   center=<23.5274 -0.2995 0.0001>
   rotation=<rot 90 -90 94.2852>
  }]
 upperarm_r [{shapeType=Capsule, length=14.9406, radius=4.1633, shapeName=upperarm_r_capsule, center=<-23.5274 0.2995 -0.0001>, rotation=<rot 90 0 4.2852>}]
}
constraints [
 {bone1Name=spine_05, bone2Name=spine_04, swing1LimitDegrees=45, swing2LimitDegrees=45, twistLimitDegrees=45}
 {bone1Name=calf_l, bone2Name=thigh_l, swing1LimitDegrees=45, swing2LimitDegrees=45, twistLimitDegrees=45}
 {
  bone1Name=spine_04
  bone2Name=spine_02
  swing1LimitDegrees=45
  swing1Motion=Free
  swing2LimitDegrees=45
  swing2Motion=Free
  twistLimitDegrees=45
  twistMotion=Free
 }
 {
  bone1Name=spine_02
  bone2Name=spine_01
  swing1LimitDegrees=45
  swing1Motion=Free
  swing2LimitDegrees=45
  swing2Motion=Free
  twistLimitDegrees=45
  twistMotion=Free
 }
 {bone1Name=calf_r, bone2Name=thigh_r, swing1LimitDegrees=45, swing2LimitDegrees=45, twistLimitDegrees=45}
 {
  bone1Name=dyn_cap_a_l_MSTR
  bone2Name=spine_05
  swing1LimitDegrees=45
  swing1Motion=Free
  swing2LimitDegrees=45
  swing2Motion=Free
  twistLimitDegrees=45
  twistMotion=Free
 }
 {bone1Name=dyn_cap_a_01_l, bone2Name=dyn_cap_a_l_MSTR, swing1LimitDegrees=10, swing2LimitDegrees=20, twistLimitDegrees=2}
 {bone1Name=dyn_cap_a_02_l, bone2Name=dyn_cap_a_01_l, swing1LimitDegrees=10, swing2LimitDegrees=20, twistLimitDegrees=2}
 {bone1Name=dyn_cap_a_03_l, bone2Name=dyn_cap_a_02_l, swing1LimitDegrees=10, swing2LimitDegrees=20, twistLimitDegrees=2}
 {
  bone1Name=dyn_cap_mid_MSTR
  bone2Name=spine_05
  swing1LimitDegrees=45
  swing1Motion=Free
  swing2LimitDegrees=45
  swing2Motion=Free
  twistLimitDegrees=45
  twistMotion=Free
 }
 {bone1Name=dyn_cap_01_mid, bone2Name=dyn_cap_mid_MSTR, swing1LimitDegrees=10, swing2LimitDegrees=20, twistLimitDegrees=10}
 {bone1Name=dyn_cap_02_mid, bone2Name=dyn_cap_01_mid, swing1LimitDegrees=10, swing2LimitDegrees=20, twistLimitDegrees=10}
 {bone1Name=dyn_cap_03_mid, bone2Name=dyn_cap_02_mid, swing1LimitDegrees=10, swing2LimitDegrees=20, twistLimitDegrees=10}
 {bone1Name=dyn_cap_04_mid, bone2Name=dyn_cap_03_mid, swing1LimitDegrees=10, swing2LimitDegrees=20, twistLimitDegrees=10}
 {
  bone1Name=dyn_cap_a_r_MSTR
  bone2Name=spine_05
  swing1LimitDegrees=45
  swing1Motion=Free
  swing2LimitDegrees=45
  swing2Motion=Free
  twistLimitDegrees=45
  twistMotion=Free
 }
 {bone1Name=dyn_cap_a_01_r, bone2Name=dyn_cap_a_r_MSTR, swing1LimitDegrees=10, swing2LimitDegrees=20, twistLimitDegrees=10}
 {bone1Name=dyn_cap_a_02_r, bone2Name=dyn_cap_a_01_r, swing1LimitDegrees=10, swing2LimitDegrees=20, twistLimitDegrees=10}
 {bone1Name=dyn_cap_a_03_r, bone2Name=dyn_cap_a_02_r, swing1LimitDegrees=10, swing2LimitDegrees=20, twistLimitDegrees=10}
 {
  bone1Name=upperarm_r
  bone2Name=spine_05
  swing1LimitDegrees=45
  swing1Motion=Free
  swing2LimitDegrees=45
  swing2Motion=Free
  twistLimitDegrees=45
  twistMotion=Free
 }
 {
  bone1Name=spine_01
  bone2Name=pelvis
  swing1LimitDegrees=45
  swing1Motion=Free
  swing2LimitDegrees=45
  swing2Motion=Free
  twistLimitDegrees=45
  twistMotion=Free
 }
 {bone1Name=dyn_cap_05_mid, bone2Name=dyn_cap_04_mid, swing1LimitDegrees=10, swing2LimitDegrees=20, twistLimitDegrees=10}
 {bone1Name=dyn_cap_a_04_r, bone2Name=dyn_cap_a_03_r, swing1LimitDegrees=10, swing2LimitDegrees=20, twistLimitDegrees=10}
 {bone1Name=dyn_cap_a_04_l, bone2Name=dyn_cap_a_03_l, swing1LimitDegrees=10, swing2LimitDegrees=20, twistLimitDegrees=2}
 {bone1Name=dyn_cap_a_04_r, bone2Name=dyn_cap_05_mid, swing1LimitDegrees=10, swing2LimitDegrees=10, twistLimitDegrees=45, twistMotion=Free}
 {bone1Name=dyn_cap_a_04_l, bone2Name=dyn_cap_05_mid, swing1LimitDegrees=10, swing2LimitDegrees=10, twistLimitDegrees=45, twistMotion=Free}
 {
  bone1Name=upperarm_l
  bone2Name=spine_05
  swing1LimitDegrees=45
  swing1Motion=Free
  swing2LimitDegrees=45
  swing2Motion=Free
  twistLimitDegrees=45
  twistMotion=Free
 }
 {
  bone1Name=lowerarm_l
  bone2Name=upperarm_l
  swing1LimitDegrees=45
  swing1Motion=Free
  swing2LimitDegrees=45
  swing2Motion=Free
  twistLimitDegrees=45
  twistMotion=Free
 }
 {
  bone1Name=lowerarm_r
  bone2Name=upperarm_r
  swing1LimitDegrees=45
  swing1Motion=Free
  swing2LimitDegrees=45
  swing2Motion=Free
  twistLimitDegrees=45
  twistMotion=Free
 }
 {
  bone1Name=hand_r
  bone2Name=lowerarm_r
  swing1LimitDegrees=45
  swing1Motion=Free
  swing2LimitDegrees=45
  swing2Motion=Free
  twistLimitDegrees=45
  twistMotion=Free
 }
 {
  bone1Name=dyn_pouches_l_MSTR
  bone2Name=thigh_l
  swing1LimitDegrees=45
  swing1Motion=Free
  swing2LimitDegrees=45
  swing2Motion=Free
  twistLimitDegrees=45
  twistMotion=Free
 }
 {bone1Name=dyn_straps_a_l_01, bone2Name=dyn_pouches_l_MSTR, swing1LimitDegrees=15, swing2LimitDegrees=10, twistLimitDegrees=2}
 {bone1Name=dyn_straps_a_l_02, bone2Name=dyn_straps_a_l_01, swing1LimitDegrees=30, swing2LimitDegrees=20, twistLimitDegrees=2}
 {bone1Name=dyn_straps_a_l_03, bone2Name=dyn_straps_a_l_02, swing1LimitDegrees=30, swing2LimitDegrees=20, twistLimitDegrees=2}
 {bone1Name=dyn_straps_b_l_01, bone2Name=dyn_pouches_l_MSTR, swing1LimitDegrees=15, swing2LimitDegrees=10, twistLimitDegrees=2}
 {bone1Name=dyn_straps_b_l_02, bone2Name=dyn_straps_b_l_01, swing1LimitDegrees=30, swing2LimitDegrees=20, twistLimitDegrees=2}
 {bone1Name=dyn_straps_b_l_03, bone2Name=dyn_straps_b_l_02, swing1LimitDegrees=30, swing2LimitDegrees=20, twistLimitDegrees=2}
 {
  bone1Name=dyn_pouches_r_MSTR
  bone2Name=thigh_r
  swing1LimitDegrees=45
  swing1Motion=Free
  swing2LimitDegrees=45
  swing2Motion=Free
  twistLimitDegrees=45
  twistMotion=Free
 }
 {bone1Name=dyn_straps_a_r_01, bone2Name=dyn_pouches_r_MSTR, swing1LimitDegrees=15, swing2LimitDegrees=10, twistLimitDegrees=2}
 {bone1Name=dyn_straps_a_r_02, bone2Name=dyn_straps_a_r_01, swing1LimitDegrees=30, swing2LimitDegrees=20, twistLimitDegrees=2}
 {bone1Name=dyn_straps_a_r_03, bone2Name=dyn_straps_a_r_02, swing1LimitDegrees=30, swing2LimitDegrees=20, twistLimitDegrees=2}
 {bone1Name=dyn_straps_b_r_01, bone2Name=dyn_pouches_r_MSTR, swing1LimitDegrees=15, swing2LimitDegrees=10, twistLimitDegrees=2}
 {bone1Name=dyn_straps_b_r_02, bone2Name=dyn_straps_b_r_01, swing1LimitDegrees=30, swing2LimitDegrees=20, twistLimitDegrees=2}
 {bone1Name=dyn_straps_b_r_03, bone2Name=dyn_straps_b_r_02, swing1LimitDegrees=30, swing2LimitDegrees=20, twistLimitDegrees=2}
 {bone1Name=thigh_l, bone2Name=pelvis, swing1LimitDegrees=45, swing2LimitDegrees=45, twistLimitDegrees=45}
 {
  bone1Name=thigh_r
  bone2Name=pelvis
  swing1LimitDegrees=45
  swing1Motion=Free
  swing2LimitDegrees=45
  swing2Motion=Free
  twistLimitDegrees=45
  twistMotion=Free
 }
]
