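# Physics asset for the dynamic floating cape on the female medium Ghost Hunter Armor Dark body.
# Two five-bone capsule chains hang from left and right master bones under a shared AIM bone on spine_03, with tight swing and twist limits along each chain.
# Cross constraints with free angular motion tie the halves at the third and fifth segments.
# Capsules on pelvis, spine, legs and thigh pouches supply the surrounding body shapes.
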PhysicsAsset /BRCosmetics/Characters/Player/Female/Medium/Bodies/F_MED_Ghost_Hunter_Armor_Dark/Meshes/F_MED_Ghost_Hunter_Armor_Dark_dyn_cape_float
bodies {
 calf_l [{shapeType=Capsule, length=47.3984, radius=6.7541, center=<-26.7539 -0.3626 -0.6294>, rotation=<rot 90 -1.7899 2.1196>}]
 calf_r [{shapeType=Capsule, length=45.8984, radius=8.2541, center=<26.7539 0.3626 0.6294>, rotation=<rot 90 0 -176.0905>}]
 dyn_ft_cape_01_l [{shapeType=Capsule, length=4, radius=6, shapeName=dyn_cap_a_01_l_capsule, center=<6 0 2.7>, rotation=<rot 0 0 90>}]
 dyn_ft_cape_01_r [{shapeType=Capsule, length=4, radius=6, shapeName=dyn_cap_a_01_l_capsule, center=<6 -0.0004 2.7>, rotation=<rot 0 0 90>}]
 dyn_ft_cape_02_l [{shapeType=Capsule, length=2.25, radius=6, shapeName=dyn_cap_a_02_l_capsule, center=<8 0 2.3>, rotation=<rot 90 0 0>}]
 dyn_ft_cape_02_r [{shapeType=Capsule, length=2.25, radius=6, shapeName=dyn_cap_a_02_l_capsule, center=<8 1 2.3>, rotation=<rot 90 0 0>}]
 dyn_ft_cape_03_l [{shapeType=Capsule, length=2.5, radius=5.75, shapeName=dyn_cap_a_03_l_capsule, center=<8 0 2.5>, rotation=<rot 90 0 0>}]
 dyn_ft_cape_03_r [{shapeType=Capsule, length=2.5, radius=5.75, shapeName=dyn_cap_a_03_l_capsule, center=<8 1 2.5>, rotation=<rot 90 0 0>}]
 dyn_ft_cape_04_l [{shapeType=Capsule, length=3.25, radius=5, shapeName=dyn_cap_a_04_l_capsule, center=<8 -2 2.5>, rotation=<rot 90 0 0>}]
 dyn_ft_cape_04_r [{shapeType=Capsule, length=3.25, radius=5, shapeName=dyn_cap_a_04_l_capsule, center=<8 2 2.5>, rotation=<rot 90 0 0>}]
 dyn_ft_cape_05_l [{shapeType=Capsule, length=3.25, radius=5, shapeName=dyn_cap_a_04_l_capsule, center=<8 -2 2.5>, rotation=<rot 90 0 0>}]
 dyn_ft_cape_05_r [{shapeType=Capsule, length=3.25, radius=5, shapeName=dyn_cap_a_04_l_capsule, center=<8 2 2.5>, rotation=<rot 90 0 0>}]
 dyn_ft_cape_AIM [{shapeType=Capsule, length=0.505, radius=0.505, shapeName=dyn_cap_mid_MSTR_capsule, rotation=<rot 0 -90 0>}]
 dyn_ft_cape_l_MSTR [{shapeType=Capsule, length=0.505, radius=0.505, shapeName=dyn_cap_mid_MSTR_capsule, rotation=<rot 0 -90 0>}]
 dyn_ft_cape_r_MSTR [{shapeType=Capsule, length=0.505, radius=0.505, shapeName=dyn_cap_mid_MSTR_capsule, rotation=<rot 0 -90 0>}]
 dyn_pouches_l_MSTR [{shapeType=Capsule, length=14.1986, radius=7.573, center=<7.8397 1.5107 -1.23>, rotation=<rot 83.962 -170.4905 179.9999>}]
 dyn_pouches_r_MSTR [{shapeType=Capsule, length=14.1986, radius=7.573, center=<7.8178 -1.507 -1.151>, rotation=<rot 85.4321 170.4977 180>}]
 pelvis [{shapeType=Capsule, length=10.0314, radius=10.5046, center=<-4.1253 -2.6077 0>}]
 spine_01 [{shapeType=Capsule, length=4.783, radius=9.7291, center=<0 -0.9625 0.0003>}]
 spine_02 [{shapeType=Capsule, length=4.7727, radius=9.7394, center=<-1.989 -0.8625 0.0003>}]
 spine_03 [{shapeType=Capsule, length=6.4469, radius=10.128, center=<-2.7071 -1.6167 0>}]
 thigh_l [{shapeType=Capsule, length=24.7059, radius=15.5657, center=<-23.3069 -6.1329 6.1946>, rotation=<rot 85.9267 -179.0662 178.7835>}]
 thigh_r [{shapeType=Capsule, length=24.7059, radius=15.5657, center=<23.3069 6.1329 -6.1946>, rotation=<rot 85.9267 -179.0662 -1.2165>}]
}
constraints [
 {
  bone1Name=spine_02
  bone2Name=spine_01
  swing1LimitDegrees=45
  swing1Motion=Free
  swing2LimitDegrees=45
  swing2Motion=Free
  twistLimitDegrees=45
  twistMotion=Free
 }
 {bone1Name=calf_r, bone2Name=thigh_r, swing1LimitDegrees=45, swing2LimitDegrees=45, twistLimitDegrees=45}
 {
  bone1Name=dyn_ft_cape_l_MSTR
  bone2Name=dyn_ft_cape_AIM
  swing1LimitDegrees=45
  swing1Motion=Free
  swing2LimitDegrees=45
  swing2Motion=Free
  twistLimitDegrees=45
  twistMotion=Free
 }
 {bone1Name=dyn_ft_cape_01_l, bone2Name=dyn_ft_cape_l_MSTR, swing1LimitDegrees=5, swing2LimitDegrees=15, twistLimitDegrees=2}
 {bone1Name=dyn_ft_cape_02_l, bone2Name=dyn_ft_cape_01_l, swing1LimitDegrees=5, swing2LimitDegrees=15, twistLimitDegrees=2}
 {bone1Name=dyn_ft_cape_03_l, bone2Name=dyn_ft_cape_02_l, swing1LimitDegrees=5, swing2LimitDegrees=15, twistLimitDegrees=2}
 {bone1Name=dyn_ft_cape_04_l, bone2Name=dyn_ft_cape_03_l, swing1LimitDegrees=5, swing2LimitDegrees=15, twistLimitDegrees=2}
 {bone1Name=dyn_ft_cape_05_l, bone2Name=dyn_ft_cape_04_l, swing1LimitDegrees=5, swing2LimitDegrees=15, twistLimitDegrees=2}
 {
  bone1Name=dyn_ft_cape_r_MSTR
  bone2Name=dyn_ft_cape_AIM
  swing1LimitDegrees=45
  swing1Motion=Free
  swing2LimitDegrees=45
  swing2Motion=Free
  twistLimitDegrees=45
  twistMotion=Free
 }
 {bone1Name=dyn_ft_cape_01_r, bone2Name=dyn_ft_cape_r_MSTR, swing1LimitDegrees=5, swing2LimitDegrees=15, twistLimitDegrees=2}
 {bone1Name=dyn_ft_cape_02_r, bone2Name=dyn_ft_cape_01_r, swing1LimitDegrees=5, swing2LimitDegrees=15, twistLimitDegrees=2}
 {bone1Name=dyn_ft_cape_03_r, bone2Name=dyn_ft_cape_02_r, swing1LimitDegrees=5, swing2LimitDegrees=15, twistLimitDegrees=2}
 {bone1Name=dyn_ft_cape_04_r, bone2Name=dyn_ft_cape_03_r, swing1LimitDegrees=5, swing2LimitDegrees=15, twistLimitDegrees=2}
 {bone1Name=dyn_ft_cape_05_r, bone2Name=dyn_ft_cape_04_r, swing1LimitDegrees=5, swing2LimitDegrees=15, twistLimitDegrees=2}
 {
  bone1Name=spine_01
  bone2Name=pelvis
  swing1LimitDegrees=45
  swing1Motion=Free
  swing2LimitDegrees=45
  swing2Motion=Free
  twistLimitDegrees=45
  twistMotion=Free
 }
 {bone1Name=calf_l, bone2Name=thigh_l, swing1LimitDegrees=45, swing2LimitDegrees=45, twistLimitDegrees=45}
 {
  bone1Name=dyn_ft_cape_AIM
  bone2Name=spine_03
  swing1LimitDegrees=45
  swing1Motion=Free
  swing2LimitDegrees=45
  swing2Motion=Free
  twistLimitDegrees=45
  twistMotion=Free
 }
 {
  bone1Name=dyn_ft_cape_05_r
  bone2Name=dyn_ft_cape_05_l
  swing1LimitDegrees=10
  swing1Motion=Free
  swing2LimitDegrees=45
  swing2Motion=Free
  twistLimitDegrees=45
  twistMotion=Free
 }
 {
  bone1Name=dyn_pouches_l_MSTR
  bone2Name=thigh_l
  swing1LimitDegrees=45
  swing1Motion=Free
  swing2LimitDegrees=45
  swing2Motion=Free
  twistLimitDegrees=45
  twistMotion=Free
 }
 {
  bone1Name=dyn_pouches_r_MSTR
  bone2Name=thigh_r
  swing1LimitDegrees=45
  swing1Motion=Free
  swing2LimitDegrees=45
  swing2Motion=Free
  twistLimitDegrees=45
  twistMotion=Free
 }
 {
  bone1Name=dyn_ft_cape_03_l
  bone2Name=dyn_ft_cape_03_r
  swing1LimitDegrees=15
  swing1Motion=Free
  swing2LimitDegrees=45
  swing2Motion=Free
  twistLimitDegrees=45
  twistMotion=Free
 }
 {
  bone1Name=spine_03
  bone2Name=spine_02
  swing1LimitDegrees=45
  swing1Motion=Free
  swing2LimitDegrees=45
  swing2Motion=Free
  twistLimitDegrees=45
  twistMotion=Free
 }
 {bone1Name=thigh_l, bone2Name=pelvis, swing1LimitDegrees=45, swing2LimitDegrees=45, twistLimitDegrees=45}
 {
  bone1Name=thigh_r
  bone2Name=pelvis
  swing1LimitDegrees=45
  swing1Motion=Free
  swing2LimitDegrees=45
  swing2Motion=Free
  twistLimitDegrees=45
  twistMotion=Free
 }
]
