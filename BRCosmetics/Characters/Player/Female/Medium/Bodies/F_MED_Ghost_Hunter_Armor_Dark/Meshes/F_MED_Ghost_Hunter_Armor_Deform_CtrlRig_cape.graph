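# Control Rig for the female Ghost Hunter armor cape that applies corrective bone offsets as the upper arms move.
# It converts each upper arm's rotation relative to spine_05 into Euler angles, remapped into clamped weights.
# Three chained additive local modifies per side shift and rotate cape master and arm deform bones.
# Each side's third weight takes the minimum of two opposing remaps, peaking mid-range and fading either way.
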ControlRigBlueprint /BRCosmetics/Characters/Player/Female/Medium/Bodies/F_MED_Ghost_Hunter_Armor_Dark/Meshes/F_MED_Ghost_Hunter_Armor_Deform_CtrlRig_cape
graphs [
 RigVMFunctionLibrary {links [], nodes []}
 RigVMModel {nodes [
   RigUnit_BeginExecution = &RigUnit_BeginExecution()
   n1 = &RigVMFunction_MathQuaternionToEuler(RotationOrder=YZX, Value=&RigUnit_GetRelativeBoneTransform(Bone=upperarm_r, Space=spine_05).Transform.Rotation)
   /Script/ControlRig.RigUnit_ModifyBoneTransforms(
    BoneToModify=((Bone="dyn_cap_a_r_MSTR",Transform=(Rotation=(X=0,Y=0,Z=0,W=1),Translation=(X=0,Y=0,Z=4),Scale3D=(X=1,Y=1,Z=1))),(Bone="deform_cape_arm_r",Transform=(Rotation=(X=0,Y=0,Z=0,W=1),Translation=(X=0,Y=0,Z=-5),Scale3D=(X=1,Y=1,Z=1))))
    Mode=AdditiveLocal
    WeightMaximum=1
    Weight=&RigVMFunction_MathFloatRemap(
     SourceMaximum=20
     SourceMinimum=45
     TargetMinimum=0
     bClamp=true
     Value=/Script/RigVM.RigVMFunction_MathFloatAbs(Value=@n1.Result.Y) {is=struct RigVMUnitNode, template="Absolute::Execute(in Value,out Result)"}.Result
    ).Result
    BoneToModify.0.Transform.Rotation=&RigVMFunction_MathQuaternionFromEuler(Euler=(X=0,Y=0,Z=0)).Result
    BoneToModify.1.Transform.Rotation=&RigVMFunction_MathQuaternionFromEuler(Euler=(X=0,Y=10,Z=0)).Result
    ExecutePin=@RigUnit_BeginExecution.ExecutePin
   ) {is=struct RigVMUnitNode}
   n2 = &RigVMFunction_MathQuaternionToEuler(RotationOrder=ZYX, Value=&RigUnit_GetRelativeBoneTransform(Bone=upperarm_r, Space=spine_05).Transform.Rotation)
   > &RigUnit_ModifyBoneTransforms(
    BoneToModify="((Bone=\"dyn_cap_a_r_MSTR\",Transform=(Rotation=(X=0,Y=0,Z=0,W=1),Translation=(X=-10,Y=-8,Z=-8),Scale3D=(X=1,Y=1,Z=1))),(Bone=\"dyn_cap_a_r_MSTR\",Transform=(Rotation=(X=0,Y=0,Z=0,W=1),Translation=(X=0,Y=0,Z=0),Scale3D=(X=1,Y=1,Z=1))),(Bone=\"deform_cape_arm_r\" ...(truncated)"
    Mode=AdditiveLocal
    WeightMaximum=1
    Weight=&RigVMFunction_MathFloatRemap(
     SourceMaximum=125
     SourceMinimum=40
     TargetMinimum=0
     bClamp=True
     Value=&RigVMFunction_MathFloatAbs(Value=@n2.Result.Z) {template="Absolute::Execute(in Value,out Result)"}.Result
    ).Result
    BoneToModify.0.Transform.Rotation=&RigVMFunction_MathQuaternionFromEuler(Euler=(X=45,Y=0,Z=0)).Result
   )
   > &RigUnit_ModifyBoneTransforms(
    BoneToModify="((Bone=\"dyn_cap_a_r_MSTR\",Transform=(Rotation=(X=0,Y=0,Z=0,W=1),Translation=(X=0,Y=0,Z=8),Scale3D=(X=1,Y=1,Z=1))),(Bone=\"dyn_cap_mid_MSTR\",Transform=(Rotation=(X=0,Y=0,Z=0,W=1),Translation=(X=0,Y=0,Z=2),Scale3D=(X=1,Y=1,Z=1))),(Bone=\"deform_cape_arm_r\",Tra ...(truncated)"
    Mode=AdditiveLocal
    WeightMaximum=1
    Weight=/Script/RigVM.RigVMFunction_MathFloatMin(
     A=&RigVMFunction_MathFloatRemap(SourceMaximum=-30, SourceMinimum=0, TargetMinimum=0, bClamp=true, Value=@n2.Result.Z).Result
     B=&RigVMFunction_MathFloatRemap(SourceMaximum=-40, SourceMinimum=-30, TargetMaximum=0, TargetMinimum=1, bClamp=true, Value=@n2.Result.Z).Result
    ) {is=struct RigVMUnitNode, template="Minimum::Execute(in A,in B,out Result)"}.Result
   )
   n3 = &RigVMFunction_MathQuaternionToEuler(RotationOrder=YZX, Value=&RigUnit_GetRelativeBoneTransform(Bone=upperarm_l, Space=spine_05).Transform.Rotation)
   > &RigUnit_ModifyBoneTransforms(
    BoneToModify=((Bone="dyn_cap_a_L_MSTR",Transform=(Rotation=(X=0,Y=0,Z=0,W=1),Translation=(X=0,Y=0,Z=0),Scale3D=(X=1,Y=1,Z=1))),(Bone="deform_cape_arm_L",Transform=(Rotation=(X=0,Y=0,Z=0,W=1),Translation=(X=0,Y=1,Z=-5),Scale3D=(X=1,Y=1,Z=1))))
    Mode=AdditiveLocal
    WeightMaximum=1
    Weight=&RigVMFunction_MathFloatRemap(
     SourceMaximum=160
     SourceMinimum=130
     TargetMinimum=0
     bClamp=true
     Value=&RigVMFunction_MathFloatAbs(Value=@n3.Result.Y) {template="Absolute::Execute(in Value,out Result)"}.Result
    ).Result
    BoneToModify.1.Transform.Rotation=&RigVMFunction_MathQuaternionFromEuler(Euler=(X=0,Y=10,Z=0)).Result
    BoneToModify.0.Transform.Rotation=&RigVMFunction_MathQuaternionFromEuler(Euler=(X=0,Y=0,Z=0)).Result
   )
   n4 = &RigVMFunction_MathQuaternionToEuler(RotationOrder=ZYX, Value=&RigUnit_GetRelativeBoneTransform(Bone=upperarm_l, Space=spine_05).Transform.Rotation)
   > &RigUnit_ModifyBoneTransforms(
    BoneToModify="((Bone=\"dyn_cap_a_l_MSTR\",Transform=(Rotation=(X=0,Y=0,Z=0,W=1),Translation=(X=-10,Y=8,Z=2),Scale3D=(X=1,Y=1,Z=1))),(Bone=\"dyn_cap_a_r_MSTR\",Transform=(Rotation=(X=0,Y=0,Z=0,W=1),Translation=(X=0,Y=0,Z=0),Scale3D=(X=1,Y=1,Z=1))),(Bone=\"deform_cape_arm_l\",T ...(truncated)"
    Mode=AdditiveLocal
    WeightMaximum=1
    Weight=&RigVMFunction_MathFloatRemap(
     SourceMaximum=50
     SourceMinimum=175
     TargetMinimum=0
     bClamp=true
     Value=&RigVMFunction_MathFloatAbs(Value=@n4.Result.Z) {template="Absolute::Execute(in Value,out Result)"}.Result
    ).Result
    BoneToModify.0.Transform.Rotation=&RigVMFunction_MathQuaternionFromEuler(Euler=(X=-45,Y=0,Z=0)).Result
   )
   > &RigUnit_ModifyBoneTransforms(
    BoneToModify="((Bone=\"dyn_cap_a_l_MSTR\",Transform=(Rotation=(X=0,Y=0,Z=0,W=1),Translation=(X=0,Y=0,Z=12),Scale3D=(X=1,Y=1,Z=1))),(Bone=\"deform_cape_arm_l\",Transform=(Rotation=(X=0,Y=0,Z=0,W=1),Translation=(X=0,Y=4,Z=0),Scale3D=(X=1,Y=1,Z=1))),(Bone=\"dyn_cap_mid_MSTR\",Tr ...(truncated)"
    Mode=AdditiveLocal
    WeightMaximum=1
    Weight=&RigVMFunction_MathFloatMin(
     A=&RigVMFunction_MathFloatRemap(SourceMaximum=135, SourceMinimum=175, TargetMinimum=0, bClamp=true, Value=@n4.Result.Z).Result
     B=&RigVMFunction_MathFloatRemap(SourceMaximum=120, SourceMinimum=135, TargetMaximum=0, TargetMinimum=1, bClamp=true, Value=@n4.Result.Z).Result
    ) {template="Minimum::Execute(in A,in B,out Result)"}.Result
   )
   &RigVMFunction_MathFloatAbs(Value=@n4.Result.Z) {template="Absolute::Execute(in Value,out Result)"}
  ]}
]
hierarchy [
 root {loc=<0 0 0>, rot=<0 0 0 1>}
 pelvis {parent=root, loc=<0 -0.5301 98.0571>, rot=<0.0127 0.707 -0.0127 -0.707>}
 spine_01 {parent=pelvis, loc=<6.1208 -0.9328 0>, rot=<0 0 0.034 -0.9994>}
 spine_02 {parent=spine_01, loc=<6.5996 -0.2095 0>, rot=<0 0 0.0238 -0.9997>}
 spine_03 {parent=spine_02, loc=<6.4914 -0.1237 0>, rot=<0 0 0.0149 -0.9999>}
 spine_04 {parent=spine_03, loc=<6.9598 0.0794 0>, rot=<0 0 -0.0132 -0.9999>}
 spine_05 {parent=spine_04, loc=<6.9987 0 0>, rot=<0 0 -0.0015 -1>}
 clavicle_l {parent=spine_05, loc=<5.4831 -0.1053 -3.2946>, rot=<0.0465 -0.7289 -0.009 -0.6829>}
 upperarm_l {parent=clavicle_l, loc=<9.8658 -1.2136 -0.4105>, rot=<0.0164 -0.3245 -0.0269 -0.9454>}
 lowerarm_l {parent=upperarm_l, loc=<26.465 -0.8345 0.259>, rot=<0.0036 0.0161 -0.2298 -0.9731>}
 hand_l {parent=lowerarm_l, loc=<22.7355 -0.3203 -0.6106>, rot=<-0.6784 -0.0223 0.0735 -0.7307>}
 index_metacarpal_l {parent=hand_l, loc=<1.6535 0.0183 -2.96>, rot=<0.0158 -0.0124 -0.0173 -0.9996>}
 index_01_l {parent=index_metacarpal_l, loc=<8.3008 0.0357 -0.0524>, rot=<-0.0044 0.0485 0.0924 -0.9945>}
 index_02_l {parent=index_01_l, loc=<3.3587 -0.3004 0.0452>, rot=<-0.002 0.0027 0.3225 -0.9466>}
 index_03_l {parent=index_02_l, loc=<2.5578 -0.123 -0.0252>, rot=<-0.002 0.0067 0.2558 -0.9667>}
 middle_metacarpal_l {parent=hand_l, loc=<2.1856 -0.1725 -0.5984>, rot=<0.006 0.0168 -0.0329 -0.9993>}
 middle_01_l {parent=middle_metacarpal_l, loc=<7.3275 -0.2447 0.0129>, rot=<-0.0208 0.0042 0.0858 -0.9961>}
 middle_02_l {parent=middle_01_l, loc=<4.0123 0.1401 0.1175>, rot=<-0.0218 0.0258 0.3512 -0.9357>}
 middle_03_l {parent=middle_02_l, loc=<2.6077 0.0534 0.1373>, rot=<0.0082 0.0188 0.1708 -0.9851>}
 pinky_metacarpal_l {parent=hand_l, loc=<1.7616 -0.9753 3.3814>, rot=<-0.0397 0.0814 -0.069 -0.9935>}
 pinky_01_l {parent=pinky_metacarpal_l, loc=<7.1709 -0.4831 -0.1094>, rot=<-0.0654 -0.0035 0.2242 -0.9723>}
 pinky_02_l {parent=pinky_01_l, loc=<2.8274 0.0893 -0.1043>, rot=<0.0037 0.0011 0.2854 -0.9584>}
 pinky_03_l {parent=pinky_02_l, loc=<2.2046 0.0307 -0.0078>, rot=<-0.0002 0.0098 0.1445 -0.9895>}
 ring_metacarpal_l {parent=hand_l, loc=<2.2074 -0.3625 1.3638>, rot=<-0.0284 0.0722 -0.0458 -0.9959>}
 ring_01_l {parent=ring_metacarpal_l, loc=<7.0295 -0.6197 -0.1446>, rot=<-0.0489 -0.0252 0.1892 -0.9804>}
 ring_02_l {parent=ring_01_l, loc=<3.6174 0.4647 -0.1027>, rot=<0 0 0.2806 -0.9598>}
 ring_03_l {parent=ring_02_l, loc=<2.4152 0.114 0.0199>, rot=<0 0 0.1189 -0.9929>}
 thumb_01_l {parent=hand_l, loc=<1.3118 -0.3191 -2.459>, rot=<0.6546 -0.4042 0.0664 -0.6354>}
 thumb_02_l {parent=thumb_01_l, loc=<5.9336 -0.0448 0.0401>, rot=<-0.0037 0.0213 0.298 -0.9543>}
 thumb_03_l {parent=thumb_02_l, loc=<2.2223 0.3682 -0.0907>, rot=<0.0118 -0.013 0.0066 -0.9998>}
 lowerarm_twist_01_l {parent=lowerarm_l, loc=<18.316 0 0>, rot=<0 0 0 1>}
 lowerarm_twist_02_l {parent=lowerarm_l, loc=<10.6707 0 0>, rot=<0 0 0 1>}
 upperarm_twist_01_l {parent=upperarm_l, loc=<7.2135 -0.1134 0>, rot=<0 0 0 1>}
 deform_cape_arm_l {parent=upperarm_twist_01_l, loc=<4.5335 -5.6871 0.6754>, rot=<1 0 0 0>}
 deform_front_cape_arm_l {parent=upperarm_twist_01_l, loc=<0.8606 10.9572 -5.0624>, rot=<0 0 0 1>}
 upperarm_twist_02_l {parent=upperarm_l, loc=<19.7998 -0.1134 0>, rot=<0 0 0 1>}
 clavicle_r {parent=spine_05, loc=<5.4831 -0.1053 3.2946>, rot=<0.7289 0.0465 -0.6829 0.009>}
 upperarm_r {parent=clavicle_r, loc=<-9.8658 1.2136 0.4104>, rot=<0.0164 -0.3245 -0.0269 -0.9454>}
 lowerarm_r {parent=upperarm_r, loc=<-26.4651 0.8345 -0.259>, rot=<0.0036 0.0161 -0.2298 -0.9731>}
 hand_r {parent=lowerarm_r, loc=<-22.7356 0.3203 0.6106>, rot=<-0.6784 -0.0223 0.0735 -0.7307>}
 index_metacarpal_r {parent=hand_r, loc=<-1.6534 -0.0181 2.9601>, rot=<0.0158 -0.0124 -0.0173 -0.9996>}
 index_01_r {parent=index_metacarpal_r, loc=<-8.3008 -0.0357 0.0524>, rot=<-0.0044 0.0485 0.0924 -0.9945>}
 index_02_r {parent=index_01_r, loc=<-3.3589 0.3003 -0.0453>, rot=<-0.002 0.0027 0.3225 -0.9466>}
 index_03_r {parent=index_02_r, loc=<-2.5578 0.123 0.0252>, rot=<-0.002 0.0067 0.2558 -0.9667>}
 middle_metacarpal_r {parent=hand_r, loc=<-2.1856 0.1726 0.5984>, rot=<0.006 0.0168 -0.0329 -0.9993>}
 middle_01_r {parent=middle_metacarpal_r, loc=<-7.3273 0.2447 -0.0129>, rot=<-0.0208 0.0042 0.0858 -0.9961>}
 middle_02_r {parent=middle_01_r, loc=<-4.0124 -0.14 -0.1175>, rot=<-0.0218 0.0258 0.3512 -0.9357>}
 middle_03_r {parent=middle_02_r, loc=<-2.6077 -0.0533 -0.1373>, rot=<0.0082 0.0188 0.1708 -0.9851>}
 pinky_metacarpal_r {parent=hand_r, loc=<-1.7617 0.9753 -3.3814>, rot=<-0.0397 0.0814 -0.069 -0.9935>}
 pinky_01_r {parent=pinky_metacarpal_r, loc=<-7.1707 0.4832 0.1094>, rot=<-0.0654 -0.0035 0.2242 -0.9723>}
 pinky_02_r {parent=pinky_01_r, loc=<-2.8274 -0.0892 0.1043>, rot=<0.0037 0.0011 0.2854 -0.9584>}
 pinky_03_r {parent=pinky_02_r, loc=<-2.2045 -0.031 0.0078>, rot=<-0.0002 0.0098 0.1445 -0.9895>}
 ring_metacarpal_r {parent=hand_r, loc=<-2.2073 0.3627 -1.3638>, rot=<-0.0284 0.0722 -0.0458 -0.9959>}
 ring_01_r {parent=ring_metacarpal_r, loc=<-7.0297 0.6197 0.1446>, rot=<-0.0489 -0.0252 0.1892 -0.9804>}
 ring_02_r {parent=ring_01_r, loc=<-3.6171 -0.4648 0.1027>, rot=<0 0 0.2806 -0.9598>}
 ring_03_r {parent=ring_02_r, loc=<-2.4153 -0.1141 -0.0199>, rot=<0 0 0.1189 -0.9929>}
 thumb_01_r {parent=hand_r, loc=<-1.3119 0.3192 2.459>, rot=<0.6546 -0.4042 0.0664 -0.6354>}
 thumb_02_r {parent=thumb_01_r, loc=<-5.9336 0.0447 -0.0401>, rot=<-0.0037 0.0213 0.298 -0.9543>}
 thumb_03_r {parent=thumb_02_r, loc=<-2.2223 -0.3682 0.0907>, rot=<0.0118 -0.013 0.0066 -0.9998>}
 lowerarm_twist_01_r {parent=lowerarm_r, loc=<-18.3161 0 0>, rot=<0 0 0 1>}
 lowerarm_twist_02_r {parent=lowerarm_r, loc=<-10.6707 0 0>, rot=<0 0 0 1>}
 upperarm_twist_01_r {parent=upperarm_r, loc=<-7.2135 0.1134 0>, rot=<0 0 0 1>}
 deform_cape_arm_r {parent=upperarm_twist_01_r, loc=<-4.5338 5.687 -0.6751>, rot=<0 0 1 0>}
 deform_front_cape_arm_r {parent=upperarm_twist_01_r, loc=<-0.8608 -10.9572 5.0625>, rot=<0 -1 0 0>}
 upperarm_twist_02_r {parent=upperarm_r, loc=<-19.7998 0.1134 0>, rot=<0 0 0 1>}
 neck_01 {parent=spine_05, loc=<11.3138 0.1529 0>, rot=<0 0 -0.2003 -0.9797>}
 neck_02 {parent=neck_01, loc=<4.5627 0 0>, rot=<0 0 0.0504 -0.9987>}
 head {parent=neck_02, loc=<5.3914 0 0>, rot=<0 0 0.115 -0.9934>}
 dyn_back_cape_MSTR {parent=spine_05, loc=<0.1722 -10.7501 0>, rot=<0.0284 -0.7065 0.7065 -0.0284>}
 dyn_cap_mid_MSTR {parent=dyn_back_cape_MSTR, loc=<10.8459 0.0007 1.9039>, rot=<-0.0001 0.0144 0.0037 -0.9999>}
 dyn_cap_01_mid {parent=dyn_cap_mid_MSTR, loc=<0 0 0>, rot=<0 0 0 1>}
 dyn_cap_02_mid {parent=dyn_cap_01_mid, loc=<9.2699 0 0>, rot=<0 -0.0016 -0.003 -1>}
 dyn_cap_03_mid {parent=dyn_cap_02_mid, loc=<18.77 0 0>, rot=<-0.0001 -0.0223 -0.0062 -0.9997>}
 dyn_cap_04_mid {parent=dyn_cap_03_mid, loc=<17.7279 0 0>, rot=<0.0002 0.014 0 -0.9999>}
 dyn_cap_05_mid {parent=dyn_cap_04_mid, loc=<14.8724 0 0>, rot=<0 0 0 1>}
 dyn_cap_a_l_MSTR {parent=dyn_back_cape_MSTR, loc=<7.6432 17.7802 0.0156>, rot=<0.0007 0.0425 -0.017 -0.999>}
 dyn_cap_a_01_l {parent=dyn_cap_a_l_MSTR, loc=<0 0 0>, rot=<0 0 0 1>}
 dyn_cap_a_02_l {parent=dyn_cap_a_01_l, loc=<13.5427 0.0001 0>, rot=<-0.0011 -0.0337 0.0018 -0.9994>}
 dyn_cap_a_03_l {parent=dyn_cap_a_02_l, loc=<17.7062 0 0>, rot=<-0.0005 -0.0165 0.0007 -0.9999>}
 dyn_cap_a_04_l {parent=dyn_cap_a_03_l, loc=<13.8747 0 0>, rot=<0 0 0 -1>}
 dyn_cap_a_r_MSTR {parent=dyn_back_cape_MSTR, loc=<7.6428 -17.7802 0.0156>, rot=<-0.0007 0.0425 0.017 -0.999>}
 dyn_cap_a_01_r {parent=dyn_cap_a_r_MSTR, loc=<0 0 0>, rot=<0 0 0 1>}
 dyn_cap_a_02_r {parent=dyn_cap_a_01_r, loc=<13.5427 -0.0001 0>, rot=<0.0011 -0.0337 -0.0018 -0.9994>}
 dyn_cap_a_03_r {parent=dyn_cap_a_02_r, loc=<17.7066 0 0>, rot=<0.0005 -0.0165 -0.0007 -0.9999>}
 dyn_cap_a_04_r {parent=dyn_cap_a_03_r, loc=<13.8747 0 0>, rot=<0 0 0 1>}
 dyn_ft_cape_AIM {parent=spine_03, loc=<1.4369 13.3824 0>, rot=<0.0316 -0.7064 -0.7064 0.0316>}
 dyn_ft_cape_l_MSTR {parent=dyn_ft_cape_AIM, loc=<0.3382 -6.8439 -0.8835>, rot=<0.0004 0.0062 -0.0665 -0.9978>}
 dyn_ft_cape_01_l {parent=dyn_ft_cape_l_MSTR, loc=<0 0 0>, rot=<0 0 0 -1>}
 dyn_ft_cape_02_l {parent=dyn_ft_cape_01_l, loc=<11.4624 0 0>, rot=<-0.0007 -0.0071 0.0277 -0.9996>}
 dyn_ft_cape_03_l {parent=dyn_ft_cape_02_l, loc=<13.5073 0 0>, rot=<-0.0008 -0.0127 0.0122 -0.9998>}
 dyn_ft_cape_04_l {parent=dyn_ft_cape_03_l, loc=<15.2262 0 0>, rot=<-0.0001 -0.002 0.0048 -1>}
 dyn_ft_cape_05_l {parent=dyn_ft_cape_04_l, loc=<14.1571 0 0>, rot=<0 0 0.0131 -0.9999>}
 dyn_ft_cape_r_MSTR {parent=dyn_ft_cape_AIM, loc=<0.338 6.8439 -0.8835>, rot=<-0.0004 0.0062 0.0665 -0.9978>}
 dyn_ft_cape_01_r {parent=dyn_ft_cape_r_MSTR, loc=<0 0 0>, rot=<0 0 0 1>}
 dyn_ft_cape_02_r {parent=dyn_ft_cape_01_r, loc=<11.4624 0 0>, rot=<0.0007 -0.0071 -0.0277 -0.9996>}
 dyn_ft_cape_03_r {parent=dyn_ft_cape_02_r, loc=<13.5076 0 0>, rot=<0.0008 -0.0127 -0.0122 -0.9998>}
 dyn_ft_cape_04_r {parent=dyn_ft_cape_03_r, loc=<15.2262 0 0.0001>, rot=<0.0001 -0.002 -0.0048 -1>}
 dyn_ft_cape_05_r {parent=dyn_ft_cape_04_r, loc=<14.1571 0 0>, rot=<0 0 -0.0131 -0.9999>}
 thigh_l {parent=pelvis, loc=<-1.0855 -0.2109 -8.7053>, rot=<0.0009 0.0355 0.0155 -0.9993>}
 calf_l {parent=thigh_l, loc=<-42.5284 -0.0019 -0.7297>, rot=<0.0006 -0.0161 -0.0391 -0.9991>}
 calf_twist_01_l {parent=calf_l, loc=<-35.8483 -1.9197 -0.5794>, rot=<0.0237 0.0095 -0.0035 -0.9997>}
 calf_twist_02_l {parent=calf_l, loc=<-19.1546 -1.7943 -0.2638>, rot=<0.0237 0.0095 -0.0035 -0.9997>}
 foot_l {parent=calf_l, loc=<-46.7401 -2.5014 -0.592>, rot=<0.0427 -0.0238 0.0353 -0.9982>}
 ball_l {parent=foot_l, loc=<-3.5658 10.8066 -0.1544>, rot=<0.0001 -0.0001 -0.7063 -0.7079>}
 thigh_twist_01_l {parent=thigh_l, loc=<-27.952 0 0>, rot=<0.0013 0 0 -1>}
 dyn_pouches_l_MSTR {parent=thigh_twist_01_l, loc=<10.2037 2.4656 -7.3188>, rot=<0.2007 -0.9768 -0.0701 0.0246>}
 dyn_straps_a_l_01 {parent=dyn_pouches_l_MSTR, loc=<3.8761 2.4276 3.3497>, rot=<-0.087 0.0443 0.0024 -0.9952>}
 dyn_straps_a_l_02 {parent=dyn_straps_a_l_01, loc=<5.9775 0 0>, rot=<0.0286 -0.066 -0.0018 -0.9974>}
 dyn_straps_a_l_03 {parent=dyn_straps_a_l_02, loc=<8.2864 0 0>, rot=<0 -0.0332 0 -0.9994>}
 dyn_straps_b_l_01 {parent=dyn_pouches_l_MSTR, loc=<3.4746 -3.5124 3.4748>, rot=<-0.0737 0.0128 0.0096 -0.9972>}
 dyn_straps_b_l_02 {parent=dyn_straps_b_l_01, loc=<5.5322 0 0>, rot=<0.0295 -0.0656 -0.0062 -0.9974>}
 dyn_straps_b_l_03 {parent=dyn_straps_b_l_02, loc=<8.6042 0 0>, rot=<0 -0.0472 0 -0.9989>}
 thigh_r {parent=pelvis, loc=<-1.0853 -0.2109 8.7053>, rot=<0.0355 -0.0009 0.9993 0.0155>}
 calf_r {parent=thigh_r, loc=<42.5286 0.0019 0.7297>, rot=<0.0006 -0.0161 -0.0391 -0.9991>}
 calf_twist_01_r {parent=calf_r, loc=<35.8485 1.9197 0.5793>, rot=<0.0237 0.0095 -0.0035 -0.9997>}
 calf_twist_02_r {parent=calf_r, loc=<19.1546 1.7943 0.2638>, rot=<0.0237 0.0095 -0.0035 -0.9997>}
 foot_r {parent=calf_r, loc=<46.7403 2.5014 0.5919>, rot=<0.0427 -0.0238 0.0353 -0.9982>}
 ball_r {parent=foot_r, loc=<3.5658 -10.8066 0.1544>, rot=<0.0001 -0.0001 -0.7063 -0.7079>}
 thigh_twist_01_r {parent=thigh_r, loc=<27.952 0 0>, rot=<0.0013 0 0 -1>}
 dyn_pouches_r_MSTR {parent=thigh_twist_01_r, loc=<-10.2036 -2.4656 7.3188>, rot=<-0.0701 -0.0246 -0.2007 -0.9768>}
 dyn_straps_a_r_01 {parent=dyn_pouches_r_MSTR, loc=<3.8761 -2.4276 3.3498>, rot=<0.087 0.0443 -0.0024 -0.9952>}
 dyn_straps_a_r_02 {parent=dyn_straps_a_r_01, loc=<5.9775 0 0>, rot=<-0.0286 -0.066 0.0018 -0.9974>}
 dyn_straps_a_r_03 {parent=dyn_straps_a_r_02, loc=<8.2864 0 0>, rot=<0 -0.0332 0 -0.9994>}
 dyn_straps_b_r_01 {parent=dyn_pouches_r_MSTR, loc=<3.4746 3.5124 3.4748>, rot=<0.0737 0.0128 -0.0096 -0.9972>}
 dyn_straps_b_r_02 {parent=dyn_straps_b_r_01, loc=<5.5322 0 0.0001>, rot=<-0.0295 -0.0656 0.0062 -0.9974>}
 dyn_straps_b_r_03 {parent=dyn_straps_b_r_02, loc=<8.6042 0 0>, rot=<0 -0.0472 0 -0.9989>}
 dyn_ft_cape_TGT {parent=pelvis, loc=<-5.9889 12.1684 0>, rot=<-0.0198 -0.7068 -0.7068 -0.0198>}
 dyn_root_simspace {parent=root, loc=<0 0 0>, rot=<0 0 0 1>}
 dyn_root_slope {parent=root, loc=<0 0 0>, rot=<0 0 0 1>}
 dyn_crouch_collision {parent=root, loc=<0 0 0>, rot=<0 0 0 1>}
]
